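# Master translucent, unlit, two-sided material for TrailBlazer drift trail Niagara ribbons, sprites and mesh particles.
# Static switches select single or dual, optionally rotated flipbook textures, whose channels are masked or blended by lerp, multiply or add.
# An optional panning mask modulates the result, which is faded along U and scaled by particle alpha and optional depth fade for opacity.
# Particle color and dynamic parameters for brightness, saturation and color shift drive emissive.
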Material /VehicleCosmetics/DriftTrail/DriftTrail_TrailBlazer/Material/M_Trail_Master
graph {
 properties {
  bAutomaticallySetUsageInEditor=false
  bUsedWithNiagaraMeshParticles=true
  bUsedWithNiagaraRibbons=true
  bUsedWithNiagaraSprites=true
  blendMode=BLEND_Translucent
  pixelDepthOffsetMode=PDOM_Legacy
  shadingModel=MSM_Unlit
  twoSided=true
  usedShadingModels=Unlit
 }
 nodes [
  n1 = MaterialFunctionCall(Opacity In=Constant() {r=1}, DepthFade Distance=Constant() {r=255}) {materialFunction=&MF_VCVFX_DepthFade}
  n2 = ScalarParameter() {defaultValue=1, group=Pan, parameterName=PanTile}
  n3 = Multiply(A=TextureCoordinate() {uTiling=0.05, vTiling=0.25}, B=@n2)
  n4 = Multiply(A=Time(), B=ScalarParameter() {defaultValue=1, group=Pan, parameterName=PanSpeed})
  n5 = TextureCoordinate() {uTiling=0.25, vTiling=0.05}
  n6 = StaticSwitchParameter(
   True=Rotator(Coordinate=Panner(Coordinate=@n3, Time=@n4) {speedY=0.4}, Time=Constant() {r=0.5}) {speed=3.14}
   False=Panner(Coordinate=Multiply(A=@n5, B=@n2), Time=@n4) {speedX=0.4}
  ) {group=FlipBook, parameterName=RotateFlipbook}
  n7 = TextureCoordinate() {coordinateIndex=1, vTiling=0}
  n8 = TextureCoordinate() {coordinateIndex=1, uTiling=0}
  n9 = AppendVector(
   A=Multiply(A=ComponentMask(None=@n7) {r=true}, B=ScalarParameter() {defaultValue=1, group=Details, parameterName=TileU})
   B=Multiply(A=ComponentMask(None=@n8) {g=true}, B=ScalarParameter() {defaultValue=1, group=Details, parameterName=TileV})
  )
  n10 = Constant() {r=0.5}
  n11 = ScalarParameter() {defaultValue=4, group=FlipBook, parameterName=Columns}
  n12 = ScalarParameter() {defaultValue=4, group=FlipBook, parameterName=Rows}
  n13 = AppendVector(A=@n11, B=@n12)
  n14 = Divide(
   A=StaticSwitchParameter(True=Rotator(Coordinate=@n9, Time=@n10) {speed=3.14}, False=@n9) {group=FlipBook, parameterName=RotateFlipbook}
   B=@n13
  )
  n15 = AppendVector(A=Multiply(A=@n11, B=@n12), B=@n12)
  n16 = Frac(None=Multiply(A=Time(), B=ScalarParameter() {defaultValue=1, group=FlipBook, parameterName=Speed}))
  n17 = AppendVector(A=@n16, B=@n16)
  n18 = Divide(A=Floor(None=Multiply(A=@n15, B=@n17)), B=@n13)
  n19 = TextureCoordinate() {coordinateIndex=1}
  n20 = Rotator(Coordinate=@n19, Time=Constant() {r=0.5}) {speed=3.14}
  n21 = TextureSampleParameter2D(UVs=StaticSwitchParameter(True=Add(A=@n14, B=@n18), False=StaticSwitchParameter(True=@n20, False=@n19) {group=FlipBook, parameterName=RotateFlipbook}) {defaultValue=true, group=FlipBook, parameterName=FlipBook}) {group=FlipBook, parameterName=BaseTexture, texture=/VehicleCosmetics/DriftTrail/DriftTrail_TrailBlazer/Textures/T_VCVFX_Basic_Trail_Difusse}
  n22 = Multiply(
   A=ScalarParameter() {defaultValue=1, group=Blend, parameterName=Blend_01_Brightness}
   B=StaticComponentMaskParameter(None=@n21.RGB) {defaultG=true, group=Blend, parameterName=Blend_01}
  )
  n23 = TextureCoordinate() {coordinateIndex=1, vTiling=0}
  n24 = TextureCoordinate() {coordinateIndex=1, uTiling=0}
  n25 = AppendVector(
   A=Multiply(A=ComponentMask(None=@n23) {r=true}, B=ScalarParameter() {defaultValue=1, group=Details, parameterName=TileU_02})
   B=Multiply(A=ComponentMask(None=@n24) {g=true}, B=ScalarParameter() {defaultValue=1, group=Details, parameterName=TileV_02})
  )
  n26 = Constant() {r=0.5}
  n27 = ScalarParameter() {defaultValue=4, group=FlipBook, parameterName=Columns_02}
  n28 = ScalarParameter() {defaultValue=4, group=FlipBook, parameterName=Rows_02}
  n29 = AppendVector(A=@n27, B=@n28)
  n30 = Divide(
   A=StaticSwitchParameter(True=Rotator(Coordinate=@n25, Time=@n26) {speed=3.14}, False=@n25) {group=FlipBook, parameterName=RotateFlipbook_02}
   B=@n29
  )
  n31 = AppendVector(A=Multiply(A=@n27, B=@n28), B=@n28)
  n32 = Frac(None=Multiply(A=Time(), B=ScalarParameter() {defaultValue=1, group=FlipBook, parameterName=Speed_02}))
  n33 = AppendVector(A=@n32, B=@n32)
  n34 = Divide(A=Floor(None=Multiply(A=@n31, B=@n33)), B=@n29)
  n35 = TextureCoordinate() {coordinateIndex=1}
  n36 = Rotator(Coordinate=@n35, Time=Constant() {r=0.5}) {speed=3.14}
  n37 = TextureSampleParameter2D(UVs=StaticSwitchParameter(True=Add(A=@n30, B=@n34), False=StaticSwitchParameter(True=@n36, False=@n35) {group=FlipBook, parameterName=RotateFlipbook_02}) {defaultValue=true, group=FlipBook, parameterName=Flipbook_02}) {
   group=FlipBook
   parameterName=BaseTexture_02
   texture=&T_VCVFX_Basic_Trail_Difusse
  }
  n38 = Multiply(
   A=ScalarParameter() {defaultValue=1, group=Blend, parameterName=Blend_02_Brightness}
   B=StaticComponentMaskParameter(None=StaticSwitchParameter(True=@n21.RGB, False=@n37.RGB) {defaultValue=true, group=Blend, parameterName=SingleTexture}) {defaultB=true, group=Blend, parameterName=Blend_02}
  )
  n39 = DynamicParameter() {paramNames [Brightness,LerpBlend,Saturation,ColorShift]}
  n40 = Multiply(A=@n22, B=@n38)
  n41 = Add(A=@n22, B=@n38)
  n42 = StaticSwitchParameter(
   True=StaticSwitchParameter(
    True=LinearInterpolate(A=@n22, B=@n38, Alpha=@n39.LerpBlend)
    False=StaticSwitchParameter(True=@n40, False=@n41) {group=Blend, parameterName="BlendType(T=M)(F=A)"}
   ) {group=Blend, parameterName=LerpBlend}
   False=Multiply(
    A=StaticComponentMaskParameter(None=@n21.RGB) {defaultG=true, group=Blend, parameterName=RGB_Mask}
    B=ScalarParameter() {defaultValue=1, group=Blend, parameterName=RGB_Mask_Brightness}
   )
  ) {group=Blend, parameterName=Blend}
  n43 = Multiply(
   A=Multiply(
    A=ScalarParameter() {defaultValue=3, group=Pan, parameterName=PanBrightness}
    B=TextureSample(UVs=@n6) {texture=/DealershipSharedLibrary/FX/Textures/Gradient_Masks/T_VCVFX_TimeMask_A}.RGB
   )
   B=@n42
  )
  n44 = StaticSwitchParameter(True=Add(A=@n43, B=@n42), False=@n43) {group=Pan, parameterName=AddPan?}
  n45 = TextureCoordinate() {uTiling=4, vTiling=4}
  n46 = ComponentMask(None=StaticSwitchParameter(True=TextureCoordinate(), False=Frac(None=@n45)) {defaultValue=true, group=FlipBook, parameterName=FlipBook}) {r=true}
  n47 = Constant() {r=1}
  n48 = OneMinus(None=@n46)
  n49 = Multiply(
   A=Multiply(
    A=StaticSwitchParameter(True=@n44, False=@n42) {group=Pan, parameterName=PanLightning}
    B=StaticSwitchParameter(True=@n46, False=@n47) {defaultValue=true, parameterName=FadeIn?}
   )
   B=@n48
  )
  MaterialExpressionVertexColor_0 = ParticleColor()
  n50 = Multiply(A=@MaterialExpressionVertexColor_0.A, B=@n49)
  MaterialExpressionConstantClamp_0 = Saturate(None=StaticSwitchParameter(True=Multiply(A=@n1.Result, B=@n50), False=@n50) {group=Opacity, parameterName=DBA?})
  MaterialExpressionConstantClamp_2 = Saturate(None=Multiply(A=Constant() {r=0.15}, B=@MaterialExpressionVertexColor_0.RGB))
  MaterialExpressionConstantClamp_1 = Saturate(None=Add(A=LinearInterpolate(A=@MaterialExpressionVertexColor_0.RGB, B=@MaterialExpressionConstantClamp_2, Alpha=@n39.ColorShift), B=@n39.Saturation))
  MaterialExpressionConstantClamp_4 = Saturate(None=@n48)
  MaterialExpressionConstantClamp_3 = Saturate(None=@n49)
  n51 = Multiply(A=Multiply(A=@MaterialExpressionConstantClamp_4, B=Constant() {r=8}), B=@MaterialExpressionConstantClamp_3)
  MaterialFunctionCall(Emissive=Multiply(A=@MaterialExpressionConstantClamp_1, B=Multiply(A=@n51, B=@n39.Brightness))) {materialFunction=&MF_FX_EmissiveAdjustments}
 ]
}
