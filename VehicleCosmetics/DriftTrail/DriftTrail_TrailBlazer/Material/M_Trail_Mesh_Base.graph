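# Unlit, translucent, two-sided base material for the TrailBlazer drift trail, usable on Niagara meshes, ribbons and sprites.
# Zero-tiled UVs pan through cloud noise, producing hard, time-varying switches that flicker between flare-ray texture channels.
# HueShift blends two such ray mixes, tinted by particle color with a red/blue-weighted brightness lift and scaled by Brightness.
# Hardness sharpens the ray alpha, multiplied by particle alpha; the Power parameter is unused.
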Material /VehicleCosmetics/DriftTrail/DriftTrail_TrailBlazer/Material/M_Trail_Mesh_Base
graph {
 properties {
  bAutomaticallySetUsageInEditor=false
  bUsedWithNiagaraMeshParticles=true
  bUsedWithNiagaraRibbons=true
  bUsedWithNiagaraSprites=true
  blendMode=BLEND_Translucent
  pixelDepthOffsetMode=PDOM_Legacy
  shadingModel=MSM_Unlit
  twoSided=true
  usedShadingModels=Unlit
 }
 nodes [
  MaterialExpressionMeshEmitterVertexColor_0 = ParticleColor()
  n1 = Add(A=@MaterialExpressionMeshEmitterVertexColor_0.R, B=@MaterialExpressionMeshEmitterVertexColor_0.R)
  MaterialExpressionConstantClamp_1 = Saturate(None=Divide(A=Add(A=@n1, B=@MaterialExpressionMeshEmitterVertexColor_0.B), B=Constant() {r=3}))
  n2 = Constant() {r=0.05}
  n3 = Constant() {r=0.16}
  n4 = Add(A=@MaterialExpressionMeshEmitterVertexColor_0.RGB, B=Constant() {r=0.1})
  n5 = Multiply(A=@n4, B=Constant() {r=1.25})
  n6 = If(
   A=@MaterialExpressionConstantClamp_1
   B=Constant() {r=0.785}
   "A > B"=@MaterialExpressionMeshEmitterVertexColor_0.RGB
   "A == B"=@MaterialExpressionMeshEmitterVertexColor_0.RGB
   "A < B"=If(
    A=@MaterialExpressionConstantClamp_1
    B=Constant() {r=0.5}
    "A > B"=Add(A=@MaterialExpressionMeshEmitterVertexColor_0.RGB, B=@n2)
    "A == B"=@n4
    "A < B"=If(A=@MaterialExpressionConstantClamp_1, B=@n3, "A > B"=@n5, "A == B"=@n5, "A < B"=@MaterialExpressionMeshEmitterVertexColor_0.RGB)
   )
  )
  n7 = TextureSample() {texture=/DealershipSharedLibrary/FX/Textures/TexturePackage/T_VCVFX_FlareRays_RGB}
  n8 = TextureCoordinate() {uTiling=0, vTiling=0}
  n9 = Panner(Coordinate=@n8) {speedX=1, speedY=0.65}
  n10 = Panner(Coordinate=@n8) {speedX=-1, speedY=-0.3}
  n11 = AppendVector(
   A=AppendVector(
    A=TextureSample(UVs=@n9) {texture=/VehicleCosmetics/FX/Textures/Noise/T_VCVFX_Clouds_A}.R
    B=TextureSample(UVs=@n10) {texture=&T_VCVFX_Clouds_A}.R
   )
   B=TextureSample(UVs=Panner(Coordinate=@n8) {speedX=-0.5, speedY=0.3}) {texture=&T_VCVFX_Clouds_A}.R
  )
  n12 = Constant() {r=0.2}
  MaterialExpressionConstantClamp_4 = Saturate(None=Multiply(A=Subtract(A=@n11, B=@n12), B=Constant() {r=35}))
  n13 = ComponentMask(None=@MaterialExpressionConstantClamp_4) {g=true}
  n14 = TextureSample() {texture=/DealershipSharedLibrary/FX/Textures/TexturePackage/T_VCVFX_FlareRays02_RGB}
  n15 = ComponentMask(None=@MaterialExpressionConstantClamp_4) {r=true}
  n16 = ComponentMask(None=@MaterialExpressionConstantClamp_4) {b=true}
  n17 = LinearInterpolate(A=LinearInterpolate(A=@n7.G, B=@n7.G, Alpha=@n13), B=LinearInterpolate(A=@n14.R, B=@n14.R, Alpha=@n15), Alpha=@n16)
  MaterialExpressionConstantClamp_2 = Saturate(None=ConstantBiasScale(None=@n17) {bias=0, scale=2})
  n18 = TextureSample() {texture=&T_VCVFX_FlareRays_RGB}
  n19 = TextureSample() {texture=&T_VCVFX_FlareRays02_RGB}
  n20 = LinearInterpolate(
   A=LinearInterpolate(A=LinearInterpolate(A=@n18.R, B=@n18.R, Alpha=@n15), B=@n18.B, Alpha=@n13)
   B=LinearInterpolate(A=LinearInterpolate(A=@n19.R, B=@n19.R, Alpha=@n15), B=@n19.B, Alpha=@n13)
   Alpha=@n16
  )
  MaterialExpressionConstantClamp_3 = Saturate(None=ConstantBiasScale(None=@n20) {bias=0, scale=2})
  MaterialExpressionMeshEmitterDynamicParameter_0 = DynamicParameter() {paramNames [Brightness,Power,HueShift,Hardness]}
  n21 = LinearInterpolate(A=@MaterialExpressionConstantClamp_2, B=@MaterialExpressionConstantClamp_3, Alpha=@MaterialExpressionMeshEmitterDynamicParameter_0.HueShift)
  MaterialFunctionCall(Emissive=Multiply(A=Multiply(A=@n6, B=@n21), B=@MaterialExpressionMeshEmitterDynamicParameter_0.Brightness)) {materialFunction=&MF_FX_EmissiveAdjustments}
  MaterialExpressionConstantClamp_5 = Saturate(None=Multiply(
    A=@MaterialExpressionMeshEmitterDynamicParameter_0.Hardness
    B=LinearInterpolate(A=@n17, B=@n20, Alpha=@MaterialExpressionMeshEmitterDynamicParameter_0.HueShift)
   ))
  Multiply(A=@MaterialExpressionMeshEmitterVertexColor_0.A, B=@MaterialExpressionConstantClamp_5)
 ]
}
